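# Material function for the RhombCamo backpack visor that adds an animated emissive pattern to incoming material attributes, blended by an Alpha input.
# An SDF shape texture drives pulsing contour lines for shape, offset drop-shadow and background layers via MF_SDF_Lines.
# Parallax bump offsets, optional rotating-noise UV warping and camera-dependent virtual-plane mapping add depth.
# It also builds scanline, Fresnel, glow and radial masks, while a ReactivityOff parameter shifts the shape UV offset.
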MaterialFunction /BRCosmetics/Accessories/FORT_Backpacks/Backpack_M_MED_RhombCamo/Materials/MF_RhombCamoVisor_Backpack
graph {nodes [
  n1 = ScalarParameter() {defaultValue=0.25, group=RhombCamoVisor, parameterName=BgShapeLineTimeOffset, sortPriority=100}
  n2 = ScalarParameter() {defaultValue=-0.01, group=RhombCamoVisor, parameterName=BgShapeLineSpeed, sortPriority=108}
  n3 = Multiply(A=Add(A=Time(), B=@n1), B=@n2)
  n4 = NamedRerouteUsage()
  n5 = Divide(A=NamedRerouteDeclaration(None=Add(A=@n3, B=@n4)) {name=BackpackSDF, nodeColor=<rgba 0.8984 0 1 1>}, B=NamedRerouteUsage())
  Divide(A=@n5)
  n6 = GetMaterialAttributes(None=NamedRerouteUsage()) {attributeGetTypes [#1]}
  n7 = Multiply(A=NamedRerouteUsage(), B=NamedRerouteUsage())
  n8 = Multiply(
   A=Add(A=NamedRerouteUsage(), B=NamedRerouteUsage())
   B=ScalarParameter() {defaultValue=0.75, group=RhombCamoVisor, parameterName=ShapeGlobalEmissiveStrength}
  )
  n9 = Add(A=@n6.EmissiveColor, B=Multiply(A=Multiply(A=@n7, B=@n8), B=NamedRerouteUsage()))
  n10 = FunctionInput() {bUsePreviewValueAsDefault=true, inputName=Alpha, inputType=FunctionInput_Scalar, sortPriority=1, previewValue=<1 0 0 1>}
  FunctionOutput(None=SetMaterialAttributes!(MaterialAttributes=@n6.MaterialAttributes, EmissiveColor=LinearInterpolate(A=@n6.EmissiveColor, B=@n9, Alpha=@n10)) {attributeSetTypes [#1]})
  NamedRerouteDeclaration(None=FunctionInput() {inputName=Material, inputType=FunctionInput_MaterialAttributes}) {name=InMaterial, nodeColor=<rgba 0 1 0.0625 1>}
  n11 = VectorParameter() {group=RhombCamoVisor, parameterName=VP_PlaneCenter, defaultValue=<rgba -1 7.5 157 1>}
  n12 = CameraPositionWS()
  n13 = VectorParameter() {group=RhombCamoVisor, parameterName=VP_PlaneNormal, defaultValue=<rgba 0 -1 0 1>}
  n14 = DotProduct(A=Normalize(VectorInput=Subtract(A=@n11.RGB, B=@n12)), B=Normalize(VectorInput=@n13.RGB))
  NamedRerouteDeclaration(None=Saturate(None=Power(Base=@n14) {constExponent=1.5})) {name=VirtualPlaneMask, nodeColor=<rgba 1 0.1875 0 1>}
  n15 = ScalarParameter() {defaultValue=20, group=RhombCamoVisor, parameterName=ScanLineTiling1, sortPriority=300}
  n16 = ScalarParameter() {defaultValue=-1, group=RhombCamoVisor, parameterName=ScanlineSpeed1, sortPriority=301}
  n17 = ScalarParameter() {defaultValue=200, group=RhombCamoVisor, parameterName=ScanLineTiling2, sortPriority=302}
  NamedRerouteDeclaration(None=LinearInterpolate(
    B=MaterialFunctionCall!(ScanLine_Tiling_1=@n15, Scanline_PannerSpeed_1=@n16, ScanLine_Tiling=@n17, Scanline_PannerSpeed, UVs) {materialFunction=/Game/Weapons/WeaponSkins/MaterialFunction/ScanLines}.Result
    Alpha=ScalarParameter() {defaultValue=0.25, group=RhombCamoVisor, parameterName=ScanLineStrength, sortPriority=299}
   ) {constA=1}) {name=Scanlines, nodeColor=<rgba 0 1 0.8125 1>}
  n18 = ScalarParameter() {defaultValue=0.05, group=RhombCamoVisor, parameterName=BgShapeLineDistance, sortPriority=101}
  NamedRerouteDeclaration(None=@n18) {name=LineDistanceBackpack, nodeColor=<rgba 1 0 0.6797 1>}
  NamedRerouteDeclaration(None=MaterialFunctionCall(
    Plane Normal Axis=@n13.RGB
    Plane X Axis=VectorParameter() {group=RhombCamoVisor, parameterName=VP_XAxis, defaultValue=<rgba 1 0 0 1>}.RGB
    Plane Center=@n11.RGB
    UV Scale=ScalarParameter() {defaultValue=32, group=RhombCamoVisor, parameterName=VP_Scale}
   ) {materialFunction=&VirtualPlaneCoordinates}.UVs) {name=VirtualPlaneUV, nodeColor=<rgba 1 0 0.9375 1>}
  NamedRerouteDeclaration(None=SphereMask(
    A=NamedRerouteUsage()
    B=Constant2Vector() {g=0.5, r=0.5}
    Radius=ScalarParameter() {defaultValue=0.48, group=RhombCamoVisor, parameterName=PatternMaskRadius, sortPriority=501}
    Hardness=ScalarParameter() {defaultValue=0.5, group=RhombCamoVisor, parameterName=PatternMaskHardness, sortPriority=502}
   ) {hardnessPercent=0}) {name=PatternMask, nodeColor=<rgba 1 0 0.3984 1>}
  n19 = Step(
   Y=NamedRerouteDeclaration(None=ScalarParameter() {group=RhombCamoVisor, parameterName=ShapeCenterStep, sortPriority=61}) {name=ShapeCenterStep, nodeColor=<rgba 0.4297 0 1 1>}
   X=NamedRerouteUsage()
  ) {constY=0.1}
  n20 = ScalarParameter() {group=RhombCamoVisor, parameterName=TimeOffset}
  n21 = NamedRerouteDeclaration(None=ScalarParameter() {defaultValue=0.004, group=RhombCamoVisor, parameterName=ShapeLineWidthMin, sortPriority=54}) {name=ShapeLineWidthMin, nodeColor=<rgba 0 0.6686 1 1>}
  n22 = MaterialFunctionCall!(
   SDF=NamedRerouteUsage()
   TimeOffset=@n20
   LineThickness=@n21
   LineDistance=NamedRerouteDeclaration(None=ScalarParameter() {defaultValue=0.07, group=RhombCamoVisor, parameterName=ShapeLineDistance, sortPriority=50}) {name=ShapeLineDistance, nodeColor=<rgba 0 0.6686 1 1>}
   VariableThickness=NamedRerouteDeclaration(None=StaticBoolParameter() {defaultValue=true, group=RhombCamoVisor, parameterName=ShapeLineVariableThickness, sortPriority=51}) {name=ShapeLineVariableThickness, nodeColor=<rgba 0 0.4609 1 1>}
   ThicknessRangeMin=NamedRerouteDeclaration(None=ScalarParameter() {defaultValue=0.3, group=RhombCamoVisor, parameterName=ShapeLineRangeMin, sortPriority=52}) {name=ShapeLineRangeMin, nodeColor=<rgba 0 0.6686 1 1>}
   ThicknessRangeMax=NamedRerouteDeclaration(None=ScalarParameter() {group=RhombCamoVisor, parameterName=ShapeLineRangeMax, sortPriority=53}) {name=ShapeLineRangeMax, nodeColor=<rgba 0 0.6686 1 1>}
   ThicknessMin=@n21
   ThicknessMax=NamedRerouteDeclaration(None=ScalarParameter() {defaultValue=0.0275, group=RhombCamoVisor, parameterName=ShapeLineWidthMax, sortPriority=55}) {name=ShapeLineWidthMax, nodeColor=<rgba 0 0.6686 1 1>}
   Animate=NamedRerouteDeclaration(None=StaticBoolParameter() {defaultValue=true, group=RhombCamoVisor, parameterName=ShapeLineAnimate, sortPriority=56}) {name=ShapeLineAnimate, nodeColor=<rgba 0 0.7891 1 1>}
   Speed=NamedRerouteDeclaration(None=ScalarParameter() {defaultValue=0.01, group=RhombCamoVisor, parameterName=ShapeLineSpeed, sortPriority=57}) {name=ShapeLineSpeed, nodeColor=<rgba 0.8516 1 0 1>}
   Pulse=NamedRerouteDeclaration(None=StaticBoolParameter() {defaultValue=true, group=RhombCamoVisor, parameterName=ShapeLinePulse, sortPriority=58}) {name=ShapeLinePulse, nodeColor=<rgba 0 0.2266 1 1>}
   PulseFrequency=NamedRerouteDeclaration(None=ScalarParameter() {defaultValue=3, group=RhombCamoVisor, parameterName=ShapeLinePulseFrequency, sortPriority=59}) {name=ShapeLinePulseFrequency, nodeColor=<rgba 0 0.6686 1 1>}
   PulseSpeed=NamedRerouteDeclaration(None=ScalarParameter() {defaultValue=-0.65, group=RhombCamoVisor, parameterName=ShapeLinePulseSpeed, sortPriority=60}) {name=ShapeLinePulseSpeed, nodeColor=<rgba 0 0.6686 1 1>}
  ) {materialFunction=/Game/Global/MaterialFunctions/MF_SDF_Lines}
  n23 = Multiply(A=@n19, B=OneMinus(None=@n22.Lines))
  n24 = NamedRerouteDeclaration(None=Multiply(A=@n19, B=@n22.PulseMask)) {name=ShapeLinePulseMask, nodeColor=<rgba 0.8516 1 0 1>}
  n25 = NamedRerouteDeclaration(None=ScalarParameter() {defaultValue=1, group=RhombCamoVisor, parameterName=ShapeLinePulseMaskOpacity, sortPriority=61}) {name=ShapeLinePulseMaskOpacity, nodeColor=<rgba 1 0 0.6094 1>}
  n26 = NamedRerouteDeclaration(None=Multiply(A=NamedRerouteDeclaration(None=@n23) {name=ShapeLines, nodeColor=<rgba 0.7578 0 1 1>}, B=LinearInterpolate(B=@n24, Alpha=@n25) {constA=1})) {name=ShapeLineMask, nodeColor=<rgba 0.3828 0 1 1>}
  n27 = VectorParameter() {group=RhombCamoVisor, parameterName=ShapeColor, sortPriority=40, defaultValue=<rgba 1 0 0 1>}
  NamedRerouteDeclaration(None=Multiply(
    A=Multiply(A=@n26, B=@n27.RGB)
    B=ScalarParameter() {defaultValue=1, group=RhombCamoVisor, parameterName=ShapeEmisiveStrength, sortPriority=41}
   )) {name=ShapeColor, nodeColor=<rgba 0.8281 1 0 1>}
  n28 = NamedRerouteDeclaration(None=VertexInterpolator(VS=PreSkinnedPosition()).PS) {name=PreskinnedPos, nodeColor=<rgba 1 0 0.4453 1>}
  n29 = VectorParameter() {group=RhombCamoVisor, parameterName=VisorGlowPos, sortPriority=200, defaultValue=<rgba 0 0 147 1>}
  n30 = ScalarParameter() {defaultValue=15, group=RhombCamoVisor, parameterName=VisorGlowRadius, sortPriority=201}
  n31 = VectorParameter() {group=RhombCamoVisor, parameterName=VisorGlowColor, sortPriority=202, defaultValue=<rgba 0.4479 0 0.0391 1>}
  n32 = ScalarParameter() {defaultValue=0.25, group=RhombCamoVisor, parameterName=VisorGlowStrength, sortPriority=203}
  NamedRerouteDeclaration(None=Multiply(A=SphereMask(A=@n28, B=@n29.RGB, Radius=@n30) {attenuationRadius=50, hardnessPercent=0}, B=Multiply(A=@n31.RGB, B=@n32))) {name=VisorGlow, nodeColor=<rgba 1 0.9844 0 1>}
  n33 = AppendVector(
   A=ScalarParameter() {group=RhombCamoVisor, parameterName=Shape_UVOffsetX, sortPriority=62}
   B=ScalarParameter() {group=RhombCamoVisor, parameterName=Shape_UVOffsetY, sortPriority=63}
  )
  n34 = Constant2Vector() {g=-0.34, r=-0.76}
  n35 = ScalarParameter() {group=EliminationCounter_React, parameterName=ReactivityOff}
  n36 = MaterialFunctionCall!(
   UV=LinearInterpolate(A=NamedRerouteUsage(), B=NamedRerouteUsage(), Alpha=ScalarParameter() {group=RhombCamoVisor, parameterName=UseVirtualPlanePattern})
   Offset=NamedRerouteDeclaration(None=LinearInterpolate(A=@n33, B=@n34, Alpha=@n35)) {name=Shape_UVOffset, nodeColor=<rgba 0.5234 0 1 1>}
   Rotation=ScalarParameter() {group=RhombCamoVisor, parameterName=Shape_UVRotation, sortPriority=64}
   Scale=ScalarParameter() {defaultValue=1, group=RhombCamoVisor, parameterName=Shape_UVScale, sortPriority=65}
   Pivot
  ) {materialFunction=/Game/Global/MaterialFunctions/MF_UV_Transform}
  n37 = LinearInterpolate(
   A=BumpOffset(Coordinate=NamedRerouteDeclaration(None=@n36.Result) {name=ShapeUV, nodeColor=<rgba 0 0.7656 1 1>}, Height=NamedRerouteUsage()) {heightRatio=0.01}
   B=NamedRerouteUsage()
   Alpha=ScalarParameter() {group=RhombCamoVisor, parameterName=UseVirtualPlanePattern}
  )
  n38 = StaticSwitchParameter(True=Add(A=@n37, B=NamedRerouteUsage()), False=@n37) {group=RhombCamoVisor, parameterName=UseShapeWarping}
  n39 = NamedRerouteUsage()
  NamedRerouteDeclaration(None=TextureSample(UVs=@n38, Tex=@n39) {samplerSource=SSM_Wrap_WorldGroupSettings, samplerType=SAMPLERTYPE_LinearColor}.R) {name=Shape_SDF, nodeColor=<rgba 0.5703 0 1 1>}
  n40 = AppendVector(
   A=ScalarParameter() {defaultValue=0.009, group=RhombCamoVisor, parameterName=DropShadow_UVOffsetX, sortPriority=72}
   B=ScalarParameter() {group=RhombCamoVisor, parameterName=DropShadow_UVOffsetY, sortPriority=73}
  )
  NamedRerouteDeclaration(None=TextureSample(UVs=Add(A=@n40, B=@n38), Tex=@n39) {samplerSource=SSM_Wrap_WorldGroupSettings, samplerType=SAMPLERTYPE_LinearColor}.R) {name=DropShadow_SDF, nodeColor=<rgba 0.5703 0 1 1>}
  n41 = Step(Y=NamedRerouteUsage(), X=NamedRerouteUsage()) {constY=0.1}
  n42 = MaterialFunctionCall(
   SDF=NamedRerouteUsage()
   TimeOffset=@n20
   LineDistance=NamedRerouteUsage()
   VariableThickness=NamedRerouteUsage()
   ThicknessRangeMin=NamedRerouteUsage()
   ThicknessRangeMax=NamedRerouteUsage()
   ThicknessMin=NamedRerouteUsage()
   ThicknessMax=NamedRerouteUsage()
   Animate=NamedRerouteUsage()
   Speed=NamedRerouteUsage()
   Pulse=NamedRerouteUsage()
   PulseFrequency=NamedRerouteUsage()
   PulseSpeed=NamedRerouteUsage()
  ) {materialFunction=&MF_SDF_Lines}
  n43 = Multiply(A=@n41, B=OneMinus(None=@n42.Lines))
  n44 = Multiply(A=@n41, B=@n42.PulseMask)
  n45 = NamedRerouteUsage()
  n46 = NamedRerouteDeclaration(None=Multiply(
    A=NamedRerouteDeclaration(None=@n43) {name=DropShadowLines, nodeColor=<rgba 0 0.5312 1 1>}
    B=LinearInterpolate(B=@n44, Alpha=@n45) {constA=1}
   )) {name=DropShadowMask, nodeColor=<rgba 1 0 0.0938 1>}
  n47 = VectorParameter() {group=RhombCamoVisor, parameterName=DropShadowColor, sortPriority=70, defaultValue=<rgba 1 0.1674 0 1>}
  NamedRerouteDeclaration(None=Multiply(
    A=Multiply(A=@n46, B=@n47.RGB)
    B=ScalarParameter() {defaultValue=1, group=RhombCamoVisor, parameterName=DropShadowEmissiveStrength, sortPriority=71}
   )) {name=DropShadowColor, nodeColor=<rgba 0 1 0.1328 1>}
  n48 = MaterialFunctionCall(
   UV=NamedRerouteUsage()
   Offset=AppendVector(
    A=ScalarParameter() {group=RhombCamoVisor, parameterName=BgShapeUVOffsetX, sortPriority=150}
    B=ScalarParameter() {group=RhombCamoVisor, parameterName=BgShapeUVOffsetY, sortPriority=151}
   )
   Rotation=ScalarParameter() {group=RhombCamoVisor, parameterName=BgShapeUVRotation, sortPriority=152}
   Scale=ScalarParameter() {defaultValue=1, group=RhombCamoVisor, parameterName=BgShapeUVScale, sortPriority=153}
  ) {materialFunction=&MF_UV_Transform}
  n49 = ScalarParameter() {defaultValue=-0.5, group=RhombCamoVisor, parameterName=BgShapeParallaxBias, sortPriority=154}
  NamedRerouteDeclaration(None=TextureSample(UVs=BumpOffset(Coordinate=@n48.Result, Height=@n49) {heightRatio=0.01}, Tex=NamedRerouteUsage()) {samplerSource=SSM_Wrap_WorldGroupSettings, samplerType=SAMPLERTYPE_LinearColor}.R) {name=BgShapeSDF, nodeColor=<rgba 0.0781 1 0 1>}
  n50 = LinearInterpolate(A=NamedRerouteUsage(), B=NamedRerouteUsage(), Alpha=NamedRerouteUsage())
  n51 = NamedRerouteUsage()
  n52 = Add(A=NamedRerouteUsage(), B=NamedRerouteUsage())
  n53 = Multiply(A=@n51, B=Saturate(None=Ceil(None=@n52)))
  n54 = LinearInterpolate(B=NamedRerouteUsage(), Alpha=ScalarParameter() {group=RhombCamoVisor, parameterName=UseVirtualPlanePattern}) {constA=1}
  NamedRerouteDeclaration(None=LinearInterpolate(A=NamedRerouteUsage(), B=Multiply(A=@n50, B=@n51), Alpha=Multiply(A=@n53, B=@n54))) {name=VisorPattern, nodeColor=<rgba 1 0 0.7969 1>}
  n55 = Step(Y=ScalarParameter() {group=RhombCamoVisor, parameterName=BgShapeCenterStep, sortPriority=112}, X=NamedRerouteUsage()) {constY=0.1}
  n56 = ScalarParameter() {defaultValue=0.002, group=RhombCamoVisor, parameterName=BgShapeLineWidthMin, sortPriority=105}
  n57 = MaterialFunctionCall(
   SDF=NamedRerouteUsage()
   TimeOffset=@n1
   LineThickness=@n56
   LineDistance=@n18
   VariableThickness=StaticBoolParameter() {group=RhombCamoVisor, parameterName=BgShapeLineVariableThickness, sortPriority=102}
   ThicknessRangeMin=ScalarParameter() {defaultValue=0.2, group=RhombCamoVisor, parameterName=BgShapeLineRangeMin, sortPriority=103}
   ThicknessRangeMax=ScalarParameter() {group=RhombCamoVisor, parameterName=BgShapeLineRangeMax, sortPriority=104}
   ThicknessMin=@n56
   ThicknessMax=ScalarParameter() {defaultValue=0.0275, group=RhombCamoVisor, parameterName=BgShapeLineWidthMax, sortPriority=106}
   Animate=StaticBoolParameter() {defaultValue=true, group=RhombCamoVisor, parameterName=BgShapeLineAnimate, sortPriority=107}
   Speed=@n2
   Pulse=StaticBoolParameter() {defaultValue=true, group=RhombCamoVisor, parameterName=BgShapeLinePulse, sortPriority=109}
   PulseFrequency=ScalarParameter() {defaultValue=3, group=RhombCamoVisor, parameterName=BgShapeLinePulseFrequency, sortPriority=110}
   PulseSpeed=ScalarParameter() {defaultValue=-0.65, group=RhombCamoVisor, parameterName=BgShapeLinePulseSpeed, sortPriority=111}
  ) {materialFunction=&MF_SDF_Lines}
  n58 = Multiply(A=@n55, B=OneMinus(None=@n57.Lines))
  n59 = Multiply(A=@n55, B=@n57.PulseMask)
  n60 = VectorParameter() {group=RhombCamoVisor, parameterName=BgShapeColor, sortPriority=90, defaultValue=<rgba 0.5782 0.8578 1 1>}
  n61 = Multiply(A=NamedRerouteDeclaration(None=Multiply(A=@n58, B=@n59)) {name=BgShapeLines, nodeColor=<rgba 0.7578 0 1 1>}, B=@n60.RGB)
  n62 = ScalarParameter() {defaultValue=0.15, group=RhombCamoVisor, parameterName=BgShapeEmissiveStrength, sortPriority=91}
  n63 = Add(A=@n5) {constB=0.5}
  n64 = Abs(None=Subtract(A=Frac(None=@n63)) {constB=0.5})
  n65 = OneMinus(None=Saturate(None=Multiply(A=@n64) {constB=5}))
  NamedRerouteDeclaration(None=Multiply(A=Multiply(A=@n61, B=@n62), B=Power(Base=@n65) {constExponent=6})) {name=BgShapeColor, nodeColor=<rgba 0.2422 1 0 1>}
  n66 = NamedRerouteUsage()
  n67 = Constant() {r=0.5}
  n68 = ScalarParameter() {defaultValue=0.45, group=RhombCamoVisor, parameterName=ShapeMaskRadius, sortPriority=68}
  n69 = Add(A=@n68) {constB=0.1}
  n70 = NamedRerouteUsage()
  n71 = Constant()
  n72 = ScalarParameter() {defaultValue=0.5, group=RhombCamoVisor, parameterName=ShapeMaskHardness, sortPriority=69}
  NamedRerouteDeclaration(None=Multiply(
    A=SphereMask(A=@n66, B=@n67, Radius=@n69) {attenuationRadius=0.6}
    B=SphereMask(A=@n70, B=@n71, Radius=@n68, Hardness=@n72) {attenuationRadius=0.5}
   )) {name=ShapeDistanceMask, nodeColor=<rgba 0.1953 0 1 1>}
  n73 = CameraVectorWS()
  n74 = PixelNormalWS()
  NamedRerouteDeclaration(None=Power(
    Base=DotProduct(A=@n73, B=@n74)
    Exp=ScalarParameter() {defaultValue=2, group=RhombCamoVisor, parameterName=VisorFresnelExponent, sortPriority=180}
   )) {name=Fresnel, nodeColor=<rgba 0 1 0.2969 1>}
  n75 = TextureCoordinate() {coordinateIndex=1}
  n76 = TextureCoordinate()
  n77 = TextureObjectParameter(UVs=NamedRerouteDeclaration(None=StaticSwitchParameter(True=@n75, False=@n76) {group=RhombCamoVisor, parameterName=UseUV1}) {name=UVs, nodeColor=<rgba 0 0.3672 1 1>}) {
   group=RhombCamoVisor
   parameterName=Shape
   samplerType=SAMPLERTYPE_LinearColor
   sortPriority=-10
   texture=/Game/Characters/Player/Male/Medium/Bodies/M_MED_RhombCamo/Textures/T_RhomboCamo_FaceAcc_SDF
  }
  NamedRerouteDeclaration(None=@n77) {name=ShapeTexture, nodeColor=<rgba 1 0.8203 0 1>}
  n78 = NamedRerouteUsage()
  n79 = Subtract(A=Constant2Vector() {g=0.5, r=0.5}, B=NamedRerouteUsage())
  n80 = Multiply(A=Time(), B=ScalarParameter() {defaultValue=0.075, group=RhombCamoVisor, parameterName=ShapeWarpingRotationSpeed})
  n81 = TextureSample(UVs=MaterialFunctionCall(UVs=@n78, Rotation Center=@n79, "Rotation Angle (0-1)"=Frac(None=@n80)) {materialFunction=&CustomRotator}.Rotated Values) {
   samplerSource=SSM_Wrap_WorldGroupSettings
   samplerType=SAMPLERTYPE_LinearColor
   texture=/Game/Characters/Player/Male/Medium/Bodies/M_MED_RhombCamo/Skins/Cyan/Textures/T_RhomboCamo_Cyan_FaceAcc_Perlin
  }
  n82 = ConstantBiasScale(None=@n81.R) {bias=-0.5, scale=2}
  n83 = ScalarParameter() {defaultValue=0.09, group=RhombCamoVisor, parameterName=ShapeWarpingIntensity}
  n84 = SphereMask(A=@n78, B=@n79, Radius=ScalarParameter() {defaultValue=0.25, group=RhombCamoVisor, parameterName=ShapeWarpingRadius}) {attenuationRadius=0.1, hardnessPercent=0}
  NamedRerouteDeclaration(None=Multiply(A=Multiply(A=@n82, B=@n83), B=SmoothStep(Value=@n84) {constMax=0, constMin=1})) {name=Warping, nodeColor=<rgba 0.3828 1 0 1>}
  n85 = NamedRerouteUsage()
  n86 = Subtract(A=Constant2Vector() {g=0.5, r=0.5}, B=NamedRerouteUsage())
  n87 = ScalarParameter() {defaultValue=0.3, group=RhombCamoVisor, parameterName=ShapeParallaxRadius}
  n88 = SmoothStep(Value=OneMinus(None=SphereMask(A=@n85, B=@n86, Radius=@n87) {attenuationRadius=0.1, hardnessPercent=0}))
  n89 = ScalarParameter() {defaultValue=7, group=RhombCamoVisor, parameterName=ShapeParallaxMulti}
  NamedRerouteDeclaration(None=Add(A=Multiply(A=@n88, B=@n89) {constB=-5}, B=ScalarParameter() {defaultValue=-5, group=RhombCamoVisor, parameterName=ShapeParallaxBias}) {constB=-10}) {name=Parallax, nodeColor=<rgba 1 0 0.7266 1>}
  OneMinus(None=@n64)
  Power(Base=@n64)
  StaticBoolParameter() {defaultValue=true, group=RhombCamoVisor, parameterName=PulseSpeed, sortPriority=51}
 ]}
