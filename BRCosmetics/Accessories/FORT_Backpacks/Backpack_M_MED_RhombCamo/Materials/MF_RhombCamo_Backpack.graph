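# Material function for the RhombCamo backpack that adds a glowing elimination counter screen on top of the MF_RhombCamoVisor_Backpack result.
# A parameter-driven count, overridable by ReactivityOff, selects cells in a horizontal number-font texture through two custom-node UV placements, one using inverted UVs.
# Two tinted digit layers, the second offset, blend into emissive color.
# Masks confine the digits and override screen roughness and metallic, leaving other material attributes intact.
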MaterialFunction /BRCosmetics/Accessories/FORT_Backpacks/Backpack_M_MED_RhombCamo/Materials/MF_RhombCamo_Backpack
graph {nodes [
  n1 = GetMaterialAttributes(None=MaterialFunctionCall!(Material=NamedRerouteUsage(), Alpha=NamedRerouteUsage()) {materialFunction=&/MF_RhombCamoVisor_Backpack}.Result) {attributeGetTypes [
    #1
    #2
    #3
    #4
   ]}
  n2 = NamedRerouteUsage()
  n3 = NamedRerouteUsage()
  n4 = LinearInterpolate(A=@n1.EmissiveColor, B=NamedRerouteUsage(), Alpha=Saturate(None=Add(A=@n2, B=@n3)))
  n5 = Saturate(None=NamedRerouteUsage())
  n6 = ScalarParameter() {group=EliminationCounter, parameterName=ScreenRoughness}
  n7 = NamedRerouteUsage()
  FunctionOutput(None=SetMaterialAttributes!(
    MaterialAttributes=@n1.MaterialAttributes
    EmissiveColor=LinearInterpolate(A=@n1.EmissiveColor, B=@n4, Alpha=@n5)
    Roughness=LinearInterpolate(A=@n1.Roughness, B=@n6, Alpha=@n7) {constB=0}
    Metallic=LinearInterpolate(A=@n1.Metallic, Alpha=@n7)
   ) {attributeSetTypes [#1, #2, #3]}) {outputName=MaterialOutput}
  n8 = NamedRerouteDeclaration(None=LinearInterpolate(
    A=ScalarParameter() {defaultValue=2, group=EliminationCounter, parameterName=Elimination_CounterOffset}
    B=Constant() {r=1}
    Alpha=ScalarParameter() {group=EliminationCounter_React, parameterName=ReactivityOff}
   )) {name=CachedKillCount, nodeColor=<rgba 1 0.5859 0 1>}
  n9 = VectorParameter() {
   group=EliminationCounter
   parameterName=Eliminaton_CounterUV01
   channelNames {a=ScaleY, b=ScaleX, g=CenterY, r=CenterX}
   defaultValue=<rgba 0.8 0.63 0.1 0.1>
  }
  MaterialExpressionAppendVector_2 = AppendVector(A=@n9.CenterX, B=@n9.CenterX)
  MaterialExpressionAppendVector_3 = AppendVector(A=@n9.ScaleX, B=@n9.ScaleX)
  MaterialExpressionTextureCoordinate_0 = TextureCoordinate()
  n10 = Custom!(UVPoint=@MaterialExpressionAppendVector_2, UVScale=@MaterialExpressionAppendVector_3, OriginalUV=@MaterialExpressionTextureCoordinate_0) {
   code="float2 DistUV =OriginalUV-UVPoint;\r\nDistUV = DistUV/UVScale + float2(0.5,0.5);\r\n\r\nreturn DistUV;"
   description=CopyToPointUV
   outputType=CMOT_Float2
   inputs [
    {
     inputName=UVPoint
     input {expression=&self:MaterialExpressionAppendVector_2}
    }
    {
     inputName=UVScale
     input {expression=&self:MaterialExpressionAppendVector_3}
    }
    {
     inputName=OriginalUV
     input {expression=&self:MaterialExpressionTextureCoordinate_0}
    }
   ]
  }
  n11 = Clamp(None=@n10)
  n12 = Add(A=Clamp(None=Subtract(A=@n8)) {maxDefault=12}, B=ComponentMask(None=@n11) {r=true})
  n13 = NamedRerouteDeclaration(None=ScalarParameter() {defaultValue=13, group=EliminationCounter, parameterName=Elimation_MaxCounterNumber}) {name=CachedMaxNumber, nodeColor=<rgba 1 0.2344 0 1>}
  NamedRerouteDeclaration(None=AppendVector(A=Divide(A=@n12, B=@n13), B=ComponentMask(None=@n11) {g=true})) {name=CachedUV01, nodeColor=<rgba 1 0 0.9375 1>}
  n14 = Subtract(A=NamedRerouteUsage())
  n15 = VectorParameter() {
   group=EliminationCounter
   parameterName=Eliminaton_CounterUV02
   channelNames {a=ScaleY, b=ScaleX, g=CenterY, r=CenterX}
   defaultValue=<rgba 0.5 0.5 0.1 0.1>
  }
  MaterialExpressionOneMinus_1 = OneMinus(None=AppendVector(A=@n15.CenterX, B=@n15.CenterX))
  MaterialExpressionAppendVector_6 = AppendVector(A=@n15.ScaleX, B=@n15.ScaleX)
  MaterialExpressionOneMinus_0 = OneMinus(None=TextureCoordinate())
  n16 = Custom(UVPoint=@MaterialExpressionOneMinus_1, UVScale=@MaterialExpressionAppendVector_6, OriginalUV=@MaterialExpressionOneMinus_0) {
   code="float2 DistUV =OriginalUV-UVPoint;\r\nDistUV = DistUV/UVScale + float2(0.5,0.5);\r\n\r\nreturn DistUV;"
   description=CopyToPointUV
   outputType=CMOT_Float2
   inputs [
    {
     inputName=UVPoint
     input {expression=&self:MaterialExpressionOneMinus_1}
    }
    {
     inputName=UVScale
     input {expression=&self:MaterialExpressionAppendVector_6}
    }
    {
     inputName=OriginalUV
     input {expression=&self:MaterialExpressionOneMinus_0}
    }
   ]
  }
  n17 = Clamp(None=@n16)
  n18 = Divide(A=Add(A=Clamp(None=@n14) {maxDefault=12}, B=ComponentMask(None=@n17) {r=true}), B=@n13)
  n19 = ComponentMask(None=@n17) {g=true}
  NamedRerouteDeclaration(None=Clamp(None=AppendVector(A=@n18, B=@n19))) {name=CachedUV02, nodeColor=<rgba 1 0 0.9375 1>}
  n20 = NamedRerouteUsage()
  NamedRerouteDeclaration(None=NamedRerouteDeclaration(None=StaticComponentMaskParameter(None=@n20) {defaultG=true, parameterName=ScreenChannel}) {name=CachedScreen, nodeColor=<rgba 1 0 0.9609 1>}) {name=CachedRoughnessMap, nodeColor=<rgba 1 0.3047 0 1>}
  n21 = NamedRerouteUsage()
  n22 = VectorParameter() {
   group=EliminationCounter
   parameterName=SecondLayerNumberOffset
   channelNames {a="", b="", g=OffsetY, r=OffsetX}
   defaultValue=<rgba 0 0 0 1>
  }
  n23 = MaterialFunctionCall(X=@n22.OffsetX, Y=@n22.OffsetX) {materialFunction=&MakeFloat2}
  n24 = Add(A=@n21, B=@n23.Result)
  n25 = NamedRerouteUsage()
  n26 = NamedRerouteUsage()
  n27 = Add(A=@n26, B=Subtract(B=@n23.Result) {constA=0})
  n28 = NamedRerouteDeclaration(None=Add(
    A=TextureSample(UVs=@n24, Tex=@n25) {samplerSource=SSM_Wrap_WorldGroupSettings}.R
    B=TextureSample(UVs=@n27, Tex=@n25) {samplerSource=SSM_Wrap_WorldGroupSettings}.R
   )) {name=CachedCounter02, nodeColor=<rgba 0 0.5781 1 1>}
  n29 = VectorParameter() {group=EliminationCounter, parameterName=Elimination_Color02, defaultValue=<rgba 1 0.5 0 1>}
  n30 = NamedRerouteUsage()
  n31 = NamedRerouteDeclaration(None=Add(
    A=TextureSample(UVs=@n21, Tex=@n30) {samplerSource=SSM_Wrap_WorldGroupSettings}.R
    B=TextureSample(UVs=@n26, Tex=@n30) {samplerSource=SSM_Wrap_WorldGroupSettings}.R
   )) {name=CachedCounter01, nodeColor=<rgba 0 0.5781 1 1>}
  n32 = VectorParameter() {group=EliminationCounter, parameterName=Elimination_Color01, defaultValue=<rgba 2 0 0 1>}
  n33 = LinearInterpolate(
   A=Multiply(A=Multiply(A=@n28, B=@n29.RGB), B=ScalarParameter() {defaultValue=1, group=EliminationCounter, parameterName=Elimination_Intensity02})
   B=Multiply(A=Multiply(A=@n31, B=@n32.RGB), B=ScalarParameter() {defaultValue=1, group=EliminationCounter, parameterName=Elimination_Intensity01})
   Alpha=NamedRerouteUsage()
  )
  n34 = ScalarParameter() {defaultValue=1, group=EliminationCounter, parameterName=EmissiveIntensityOverall}
  NamedRerouteDeclaration(None=Multiply(A=Multiply(A=@n33, B=@n34), B=NamedRerouteUsage())) {name=CachedNumberColor, nodeColor=<rgba 1 0.4219 0 1>}
  n35 = TextureObjectParameter() {
   group=EliminationCounter
   parameterName=Counter_NumbersMap
   texture=/BRCosmetics/Accessories/FORT_Backpacks/Backpack_M_MED_RhombCamo/Textures/T_Number_Font_Center_small
  }
  NamedRerouteDeclaration(None=@n35) {name=NumberTexture, nodeColor=<rgba 0.8984 0 1 1>}
  NamedRerouteDeclaration(None=FunctionInput() {inputName=MaterialInput, inputType=FunctionInput_MaterialAttributes}) {name=CachedMaterialIn, nodeColor=<rgba 0.6172 1 0 1>}
  n36 = CameraVectorWS()
  n37 = PixelNormalWS()
  NamedRerouteDeclaration(None=Power(
    Base=DotProduct(A=@n36, B=@n37)
    Exp=ScalarParameter() {defaultValue=1, group=EliminationCounter, parameterName=VisorFresnelExponent_Counter, sortPriority=180}
   )) {name=FresnelIntensity, nodeColor=<rgba 0 1 0.2969 1>}
  NamedRerouteDeclaration(None=FunctionInput() {inputName=TechArtMaskInput, sortPriority=1}) {name=CachedTechArtMask, nodeColor=<rgba 1 0.0703 0 1>}
 ]}
