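# Animation Blueprint for the RhombCamo backpack, animating its case fold, latches and handcuff.
# The pack bone aims at the character's pelvis, read each update from the customization base mesh, plus a fixed roll.
# Tactical sprinting, skydiving or sliding blends the fold and latches closed and disables latch dynamics.
# The output uses Anim Dynamics when bAnimDynamicsEnabled is true, otherwise partial-alpha Trail Controllers, both branching from the cached case-closing pose.
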AnimBlueprint /BRCosmetics/Accessories/FORT_Backpacks/Backpack_M_MED_RhombCamo/Meshes/M_MED_RhombCamo_Pack_AnimBP
target_skeleton=&/M_MED_RhombCamo_Pack_Skeleton
graphs [
 AnimGraph {nodes [
   Output Pose(Result=Blend Poses by bool(
     BlendPose_0=Use cached pose 'PostAnimDynamicsPose'()
     BlendPose_1=Use cached pose 'PostTrailControllerPose'()
     bActiveValue=Get bAnimDynamicsEnabled() -> (bAnimDynamicsEnabled:bool)
    ) {blendType=Linear}) {layerGroup=None}
   n1 = Input Pose()
   n2 = "Transform (Modify) Bone - Bone: pack"(
    ComponentPose=Look At - Bone: pack!(ComponentPose=Local To Component(LocalPose=@n1), Alpha=1, LookAtLocation:Vector=Get PelvisLocation() -> (PelvisLocation:Vector)) -> (Pose) {alphaBoolBlend {blendOption=Linear}, boneToModify {boneName=pack}, lookAt_Axis {axis {y=0, z=-1}}, lookAtTarget=~}
    Translation="0,0,0"
    Rotation="0,0,20"
   ) {
    rotationMode=BMM_Additive
    rotationSpace=BCS_BoneSpace
    alphaBoolBlend {blendOption=Linear}
    boneToModify {boneName=pack}
    rotation {roll=20}
   }
   Save cached pose 'BasePose'(Pose=Component To Local(ComponentPose=@n2)) {cacheName=BasePose}
   n3 = Get bIsCaseShouldBeClosed() -> (bIsCaseShouldBeClosed:bool)
   n4 = "Transform (Modify) Bone - Bone: dyn_fold"!(ComponentPose=Local To Component(LocalPose=Use cached pose 'BasePose'()), bAlphaBoolEnabled:bool=@n3, Rotation="0,0,180") -> (Pose) {
    alphaInputType=Bool
    rotationMode=BMM_Additive
    alphaBoolBlend {blendInTime=0.2, blendOption=Linear, blendOutTime=0.4}
    boneToModify {boneName=dyn_fold}
    rotation {roll=180}
   }
   n5 = "Transform (Modify) Bone - Bone: dyn_latch_l"!(ComponentPose=@n4, bAlphaBoolEnabled:bool=@n3, Rotation="0,0,-40") -> (Pose) {
    alphaInputType=Bool
    rotationMode=BMM_Additive
    alphaBoolBlend {blendInTime=0.2, blendOption=Linear, blendOutTime=0.4}
    boneToModify {boneName=dyn_latch_l}
    rotation {roll=-40}
   }
   n6 = "Transform (Modify) Bone - Bone: dyn_latch_r"!(ComponentPose=@n5, bAlphaBoolEnabled:bool=@n3, Rotation="0,0,-40") -> (Pose) {
    alphaInputType=Bool
    rotationMode=BMM_Additive
    alphaBoolBlend {blendInTime=0.2, blendOption=Linear, blendOutTime=0.4}
    boneToModify {boneName=dyn_latch_r}
    rotation {roll=-40}
   }
   Save cached pose 'PostCaseClosingPose'(Pose=Component To Local(ComponentPose=@n6)) {cacheName=PostCaseClosingPose}
   n7 = Anim Dynamics - Bone: dyn_handcuff(ComponentPose=Local To Component(LocalPose=Use cached pose 'PostCaseClosingPose'()), Alpha=1, ExternalForce="0,0,0") {
    angularSpringConstant=2
    bAngularSpring=true
    bUseGravityOverride=true
    bUsePlanarLimit=false
    alphaBoolBlend {blendOption=Linear}
    boundBone {boneName=dyn_handcuff}
    componentAppliedLinearAccClamp=<100 1000 1000>
    componentLinearAccScale=<0.4 0.4 0.4>
    componentLinearVelScale=<0.4 0.4 0.4>
    gravityOverride {z=-400}
    physicsBodyDefinitions [{
      boundBone {boneName=dyn_handcuff}
      boxExtents=<10 10 7>
      constraintSetup {twistAxis=AxisZ, angularLimitsMax=<10 0 0>, angularLimitsMin=<-60 0 0>, angularTarget=<0 0 0>}
      localJointOffset=<0 0 4>
     }]
   }
   n8 = Get bIsLatchDynamicEnabled() -> (bIsLatchDynamicEnabled:bool)
   n9 = Anim Dynamics - Bone: dyn_latch_l!(ComponentPose=@n7, bAlphaBoolEnabled:bool=@n8, ExternalForce="0,0,0") -> (Pose) {
    alphaInputType=Bool
    angularDampingOverride=0.75
    angularSpringConstant=4
    bAngularSpring=true
    bOverrideAngularDamping=true
    bUseGravityOverride=true
    bUsePlanarLimit=false
    simulationSpace=BoneRelative
    alphaBoolBlend {blendInTime=1, blendOption=Linear, blendOutTime=0.1}
    boundBone {boneName=dyn_latch_l}
    componentAppliedLinearAccClamp=<300 300 300>
    componentLinearAccScale=<1 1 1>
    componentLinearVelScale=<1 1 1>
    gravityOverride {z=-150}
    physicsBodyDefinitions [{
      boundBone {boneName=dyn_latch_l}
      boxExtents=<6 8 8>
      constraintSetup {twistAxis=AxisZ, angularLimitsMax=<25 0 0>, angularLimitsMin=<-45 0 0>, angularTarget=<0 0 0>}
      localJointOffset=<0 0 8>
     }]
    relativeSpaceBone {boneName=pack}
   }
   n10 = Anim Dynamics - Bone: dyn_latch_r!(ComponentPose=@n9, bAlphaBoolEnabled:bool=@n8, ExternalForce="0,0,0") -> (Pose) {
    alphaInputType=Bool
    angularDampingOverride=0.8
    angularSpringConstant=6
    bUseGravityOverride=true
    bUsePlanarLimit=false
    simulationSpace=BoneRelative
    alphaBoolBlend {blendInTime=1, blendOption=Linear, blendOutTime=0.1}
    boundBone {boneName=dyn_latch_r}
    componentAppliedLinearAccClamp=<400 400 400>
    componentLinearAccScale=<0.5 0.5 0.5>
    componentLinearVelScale=<0.5 0.5 0.5>
    gravityOverride {z=-100}
    physicsBodyDefinitions [{
      boundBone {boneName=dyn_latch_r}
      boxExtents=<4 6 6>
      constraintSetup {twistAxis=AxisZ, angularLimitsMax=<25 0 0>, angularLimitsMin=<-50 0 0>, angularTarget=<0 0 0>}
      localJointOffset=<0 0 6>
     }]
    relativeSpaceBone {boneName=pack}
   }
   Save cached pose 'PostAnimDynamicsPose'(Pose=Component To Local(ComponentPose=@n10)) {cacheName=PostAnimDynamicsPose}
   n11 = Trail controller - Bone: dyn_handcuff_end(ComponentPose=Local To Component(LocalPose=Use cached pose 'PostCaseClosingPose'()), Alpha=0.3) {
    chainBoneAxis=Z
    relaxationSpeedScale=8
    rotationLimits [{limitMax=<0 0 0>, limitMin=<0 0 0>}, {limitMax=<10 10 10>, limitMin=<-10 -10 -10>}]
    rotationOffsets [<0 0 0>, <0 0 0>]
    trailBone {boneName=dyn_handcuff_end}
   }
   n12 = Trail controller - Bone: dyn_latch_l_end(ComponentPose=@n11, Alpha=0.2) {
    chainBoneAxis=Z
    relaxationSpeedScale=1.45
    rotationLimits [{limitMax=<180 180 180>}, {limitMax=<180 180 180>}]
    rotationOffsets [<0 0 0>, <0 0 0>]
    trailBone {boneName=dyn_latch_l_end}
    trailRelaxationSpeed {editorCurveData {keys [<at 0 30 RCIM_Linear>,<at 1 30 RCIM_Linear>]}}
   }
   n13 = Trail controller - Bone: dyn_latch_r_end(ComponentPose=@n12, Alpha=0.4) {
    chainBoneAxis=Z
    relaxationSpeedScale=1.2
    rotationLimits [{limitMax=<180 180 180>}, {limitMax=<180 180 180>}]
    rotationOffsets [<0 0 0>, <0 0 0>]
    trailBone {boneName=dyn_latch_r_end}
    trailRelaxationSpeed {editorCurveData {keys [<at 0 30 RCIM_Linear>,<at 1 30 RCIM_Linear>]}}
   }
   Save cached pose 'PostTrailControllerPose'(Pose=Component To Local(ComponentPose=@n13)) {cacheName=PostTrailControllerPose}
  ]}
 EventGraph {nodes [
   Event BlueprintUpdateAnimation()
   > n1 = Set bIsCaseShouldBeClosed(bIsCaseShouldBeClosed:bool=OR Boolean!(A=Get bIsTacticalSprinting(), B=Get bIsSkyDiving(), C:bool=Get bIsSliding()) -> (ReturnValue)) -> (Output_Get:bool)
   > Set bIsLatchDynamicEnabled(bIsLatchDynamicEnabled:bool=NOT Boolean(A=@n1)) -> (Output_Get:bool)
   > n2 = Get!(execute:Exec) -> (then:Exec, else:Exec, BaseMesh:Skeletal Mesh Component Object Reference)
   > Set PelvisLocation(PelvisLocation:Vector=GetSocketLocation!(self:Scene Component Object Reference=@n2, InSocketName:Name=pelvis) -> (ReturnValue:Vector)) -> (Output_Get:Vector)
   n3 = Event OnCharacterCustomizationLoaded()
   > Set BaseMesh(BaseMesh:Skeletal Mesh Component Object Reference=@n3.CustomBaseMesh) -> (Output_Get:Skeletal Mesh Component Object Reference)
   TryGetPawnOwner()
  ]}
]
variables [
 BaseMesh {type=SkeletalMeshComponent}
 bIsCaseShouldBeClosed {type=boolean}
 bIsLatchDynamicEnabled {type=boolean}
 PelvisLocation {type=Vector}
]
